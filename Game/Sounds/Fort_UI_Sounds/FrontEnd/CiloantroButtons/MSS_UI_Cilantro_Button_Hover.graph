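# A mono one-shot MetaSound Source for the front-end Cilantro button hover sound in the UI.
# On play, a wave player plays the exposed Default Hover Sound input, preset to a default hover wave, and reports finished when it ends.
# A concurrency override keeps a single instance, stopping the oldest, and the sound routes to the Fort_UI class and final submix with virtualization disabled.
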MetaSoundSource /Game/Sounds/Fort_UI_Sounds/FrontEnd/CiloantroButtons/MSS_UI_Cilantro_Button_Hover
inputs [
 Default Hover Sound {type=WaveAsset}
 UE.Source.OnPlay {type=Trigger}
]
interfaces [UE.OutputFormat.Mono:1.0, UE.Source.OneShot:1.0, UE.Source:1.1]
outputs [
 "UE.OutputFormat.Mono.Audio:0" {type=Audio}
 UE.Source.OneShot.OnFinished {type=Trigger}
]
pages [Default {
  input_defaults {Default Hover Sound=/Game/Sounds/Fort_UI_Sounds/FrontEnd/DefaultSounds/PressedButtons/wav/UI_MP_Hover_Default_01, UE.Source.OnPlay=null}
  nodes [WavePlayer_0 = UE.Wave Player:Mono(Wave Asset=^Default Hover Sound, Play=^UE.Source.OnPlay) {also {version [1,1]}}]
  links [{dst="@UE.Source.OneShot.OnFinished", src=WavePlayer_0.On Finished}, {dst="@UE.OutputFormat.Mono.Audio:0", src=WavePlayer_0.Out Mono}]
 }]
properties {
 assetImportData=&self:AssetImportData
 bEnableBaseSubmix=false
 bOverrideConcurrency=true
 priority=0.75
 sampleRate=48000
 soundClassObject=&Fort_UI
 virtualizationMode=Disabled
 volume=0.75
 concurrencyOverrides {bVolumeScaleCanRelease=true, maxCount=1, resolutionRule=StopOldest, volumeScale=0.5, volumeScaleAttackTime=0.25, volumeScaleReleaseTime=0.25}
 soundSubmixSends [{soundSubmix=/Game/Sounds/SubMixes/Fortnite_Final}]
}
